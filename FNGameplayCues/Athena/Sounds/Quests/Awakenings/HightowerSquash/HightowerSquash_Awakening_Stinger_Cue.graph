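# Sound cue for the HightowerSquash awakening quest stinger, played as non-spatialized audio in the in-game critical UI sound class.
# A mixer layers a reward-screen chest upgrade sound, reused from HightowerGrape and rerouted to the general UI class, with the louder Squash challenge stinger after a short fixed delay.
# Concurrency allows one instance at a time, stopping the oldest when retriggered.
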SoundCue /FNGameplayCues/Athena/Sounds/Quests/Awakenings/HightowerSquash/HightowerSquash_Awakening_Stinger_Cue
root=Mixer_0
nodes [
 {id=Delay_1, children [WavePlayer_0], properties {delayMax=0.3, delayMin=0.3}}
 {id=Mixer_0, children [SoundClass_0,Delay_1], properties {inputVolume [0.65,0.85]}}
 {id=SoundClass_0, children [WavePlayer_1], properties {soundClassOverride=&Fort_UI}}
 {
  id=WavePlayer_0
  properties {soundWaveAssetPtr=&/Hightower_Squash_Awake_Challenge_Stinger}
 }
 {
  id=WavePlayer_1
  properties {soundWaveAssetPtr=/FNGameplayCues/Athena/Sounds/Quests/Awakenings/HightowerGrape/Fort_RewardScreen_TeamScore_ChestLootUpgrade}
 }
]
properties {
 bOverrideAttenuation=true
 bOverrideConcurrency=true
 maxDistance=4000
 soundClassObject=&Fort_UI_InGame_Critical
 volumeMultiplier=1
 attenuationOverrides {bSpatialize=false, falloffDistance=2000, attenuationShapeExtents {x=2000}}
 concurrencyOverrides {maxCount=1, resolutionRule=StopOldest}
}
